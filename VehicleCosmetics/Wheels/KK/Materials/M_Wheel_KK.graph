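# Wheel material for the KK vehicle cosmetic, giving the rim a pulsing "infinity" effect in FXColor over a BaseColor tint.
# A Time-driven sine, shaped by Speed and Offset, animates the offset input of the shared MF_DLR_InfinityOffset function on the rim FX texture.
# That clamped result, scaled by FXOpacity and added to a static sample's red channel, blends base color.
# Emissive uses only the static sample times squared FXColor, so it does not pulse.
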Material /VehicleCosmetics/Wheels/KK/Materials/M_Wheel_KK
graph {
 properties {
  bAutomaticallySetUsageInEditor=false
  bUseMaterialAttributes=true
  bUsedWithMeshDeformer=true
  bUsedWithSkeletalMesh=true
  blendMode=BLEND_Masked
  pixelDepthOffsetMode=PDOM_Legacy
  usedShadingModels=Default Lit
 }
 nodes [
  n1 = VectorParameter() {
   group=0_Special Effects
   parameterName=FXColor
   sortPriority=0
   channelNames {a=Emissive, b="", g="", r=""}
   defaultValue=<rgba 1 0.4735 0.0185 1>
  }
  n2 = Time()
  n3 = ScalarParameter() {defaultValue=2, group=0_Special Effects, parameterName=Speed, sortPriority=2}
  n4 = ScalarParameter() {defaultValue=2.5, group=0_Special Effects, parameterName=Offset, sortPriority=3}
  n5 = Multiply(A=Sine(None=Multiply(A=@n2, B=@n3)), B=Subtract(B=@n4)) {constB=0.4}
  n6 = TextureObjectParameter() {
   group=0_Special Effects
   parameterName=InfinityFXTex
   samplerType=SAMPLERTYPE_LinearColor
   sortPriority=5
   texture=/VehicleCosmetics/Wheels/KK/Textures/T_Rim_KK_FX_RGB
  }
  n7 = MaterialFunctionCall!(
   bRotateUVs=StaticBool()
   bGradientFalloff=StaticBool() {value=true}
   GradientDarkness=Constant() {r=0.001}
   Offset=OneMinus(None=Add(A=@n5, B=@n4) {constB=0.6})
   UVs
   Texture=@n6
  ) {materialFunction=/DealershipSharedLibrary/Materials/MaterialFunctions/MF_DLR_InfinityOffset}
  n8 = Multiply(
   A=ComponentMask(None=Clamp(None=@n7.Quality Infinite Offset)) {r=true}
   B=ScalarParameter() {defaultValue=2, group=0_Special Effects, parameterName=FXOpacity, sortPriority=4}
  )
  n9 = TextureSampleParameter2D() {
   group=0_Special Effects
   parameterName=InfinityFXSample
   samplerType=SAMPLERTYPE_LinearColor
   sortPriority=6
   texture=&T_Rim_KK_FX_RGB
  }
  MakeMaterialAttributes!(
   BaseColor=LinearInterpolate(
    A=VectorParameter() {group=0_Special Effects, parameterName=BaseColor, sortPriority=1, defaultValue=<rgba 0 0 0 1>}.RGB
    B=@n1.RGB
    Alpha=Add(A=@n8, B=@n9.R)
   )
   Metallic
   Specular
   Roughness
   Anisotropy
   EmissiveColor=Multiply(A=@n9.R, B=Multiply(A=@n1.RGB, B=@n1.RGB))
   Opacity
   OpacityMask
   Normal
   Tangent
   WorldPositionOffset
   SubsurfaceColor
   ClearCoat
   ClearCoatRoughness
   AmbientOcclusion
   Refraction
   CustomizedUVs_0
   CustomizedUVs_1
   CustomizedUVs_2
   CustomizedUVs_3
   CustomizedUVs_4
   CustomizedUVs_5
   CustomizedUVs_6
   CustomizedUVs_7
   PixelDepthOffset
   ShadingModel
   Displacement
  )
 ]
}
